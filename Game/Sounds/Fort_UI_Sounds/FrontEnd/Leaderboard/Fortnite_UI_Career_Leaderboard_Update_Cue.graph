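# Sound cue for the front-end career leaderboard update sound in the UI.
# A single wave player plays the career page tile slide-down sound wave directly as the root node.
# It routes through the Fort_UI sound class at reduced volume, with no randomization or other processing nodes.
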SoundCue /Game/Sounds/Fort_UI_Sounds/FrontEnd/Leaderboard/Fortnite_UI_Career_Leaderboard_Update_Cue
root=WavePlayer_0
nodes [{id=WavePlayer_0, properties {soundWaveAssetPtr=/Game/Sounds/Fort_UI_Sounds/FrontEnd/CareerPage/UI_FrontEnd_Career_TileSlideDown}}]
properties {maxDistance=2097152, soundClassObject=&Fort_UI, volumeMultiplier=0.5}
